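# Turtle dove bird one-shot Sound Cue for outdoor ambient spots at the Mount Olympus POI.
# A random node picks one of three branches, a finite loop of three, a loop of two, or a single call, each randomizing among the same four dove waves.
# A modulator varies the result, and every quality level shares that chain.
# Playback is quiet, uses bird one-shot attenuation, and sends to the exterior ambient source bus.
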SoundCue /Game/Sounds/Helios/POI/MountOlympus/Spots/Spot_Bird_Olympus_Turtle_Dove_Cue
root=QualityLevel_0
nodes [
 {id=Looping_0, children [Random_1], properties {bLoopIndefinitely=false, loopCount=3}}
 {id=Looping_2, children [Random_4], properties {bLoopIndefinitely=false, loopCount=2}}
 {id=Modulator_0, children [Random_6]}
 {id=QualityLevel_0, children [Modulator_0,Modulator_0,Modulator_0,Modulator_0]}
 {id=Random_1, children [WavePlayer_0,WavePlayer_1,WavePlayer_2,WavePlayer_3], properties {weights [1,1,1,1]}}
 {id=Random_4, children [WavePlayer_4,WavePlayer_5,WavePlayer_6,WavePlayer_7], properties {weights [1,1,1,1]}}
 {id=Random_5, children [WavePlayer_8, WavePlayer_9, WavePlayer_10, WavePlayer_11], properties {weights [1,1,1,1]}}
 {id=Random_6, children [Looping_0,Looping_2,Random_5], properties {weights [1,1,1]}}
 {id=WavePlayer_0, properties {soundWaveAssetPtr=&/Spot_Bird_Turtle_Dove_01}}
 {id=WavePlayer_1, properties {soundWaveAssetPtr=&/Spot_Bird_Turtle_Dove_02}}
 {id=WavePlayer_10, properties {soundWaveAssetPtr=&/Spot_Bird_Turtle_Dove_03}}
 {id=WavePlayer_11, properties {soundWaveAssetPtr=&/Spot_Bird_Turtle_Dove_04}}
 {id=WavePlayer_2, properties {soundWaveAssetPtr=&Spot_Bird_Turtle_Dove_03}}
 {id=WavePlayer_3, properties {soundWaveAssetPtr=&Spot_Bird_Turtle_Dove_04}}
 {id=WavePlayer_4, properties {soundWaveAssetPtr=&Spot_Bird_Turtle_Dove_01}}
 {id=WavePlayer_5, properties {soundWaveAssetPtr=&Spot_Bird_Turtle_Dove_02}}
 {id=WavePlayer_6, properties {soundWaveAssetPtr=&Spot_Bird_Turtle_Dove_03}}
 {id=WavePlayer_7, properties {soundWaveAssetPtr=&Spot_Bird_Turtle_Dove_04}}
 {id=WavePlayer_8, properties {soundWaveAssetPtr=&Spot_Bird_Turtle_Dove_01}}
 {id=WavePlayer_9, properties {soundWaveAssetPtr=&Spot_Bird_Turtle_Dove_02}}
]
properties {
 attenuationSettings=&Fort_TileAmbient_OneShot_Bird_Att
 maxDistance=8005
 soundClassObject=&Fort_Ambient_OneShots_Outdoors
 volumeMultiplier=0.2
 busSends [{
   audioBus=null
   maxSendDistance=6000
   minSendDistance=3000
   minSendLevel=0.6
   sendLevel=1
   soundSourceBus=&AmbientAudio_Exterior_SourceBus
  }]
}
