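# Ambient sound cue that plays sporadic bronze tone one-shots as an exterior spot sound at the Mount Olympus POI.
# A quality switch sends its first two levels to a six-tone random pool and its last two to a two-tone pool.
# Each branch loops a random delay into a random pick, with pitch and volume modulation.
# The cue is pitched down and quiet, using musical one-shot attenuation and its own concurrency limit.
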SoundCue /Game/Sounds/Helios/POI/MountOlympus/Spots/Spot_Olympus_Bronze_Oneshots_Ext_Loop_Cue
root=QualityLevel_0
nodes [
 {id=Delay_1, children [Random_2], properties {delayMax=20, delayMin=5}}
 {id=Delay_2, children [Random_3], properties {delayMax=20, delayMin=5}}
 {id=Looping_0, children [Delay_1]}
 {id=Looping_1, children [Delay_2]}
 {id=Modulator_0, children [Looping_1], properties {pitchMax=1.1, pitchMin=0.9, volumeMax=0.9, volumeMin=0.8}}
 {id=Modulator_1, children [Looping_0], properties {pitchMax=1.1, pitchMin=0.9, volumeMax=1.1, volumeMin=0.9}}
 {id=QualityLevel_0, children [Modulator_1,Modulator_1,Modulator_0,Modulator_0]}
 {id=Random_2, children [WavePlayer_0,WavePlayer_1,WavePlayer_2,WavePlayer_3,WavePlayer_4,WavePlayer_5], properties {weights [1,1,1,1,1,1]}}
 {id=Random_3, children [WavePlayer_6,WavePlayer_7], properties {weights [1,1]}}
 {id=WavePlayer_0, properties {soundWaveAssetPtr=&/Spot_Olympus_Bronze_Tone_01}}
 {id=WavePlayer_1, properties {soundWaveAssetPtr=&/Spot_Olympus_Bronze_Tone_02}}
 {id=WavePlayer_2, properties {soundWaveAssetPtr=&/Spot_Olympus_Bronze_Tone_03}}
 {id=WavePlayer_3, properties {soundWaveAssetPtr=&/Spot_Olympus_Bronze_Tone_04}}
 {id=WavePlayer_4, properties {soundWaveAssetPtr=&/Spot_Olympus_Bronze_Tone_05}}
 {id=WavePlayer_5, properties {soundWaveAssetPtr=&/Spot_Olympus_Bronze_Tone_06}}
 {id=WavePlayer_6, properties {soundWaveAssetPtr=&Spot_Olympus_Bronze_Tone_02}}
 {id=WavePlayer_7, properties {soundWaveAssetPtr=&Spot_Olympus_Bronze_Tone_01}}
]
properties {
 attenuationSettings=/Game/Sounds/Attenuation/Fort_Ambient_Athena_Musical_Oneshot_NoVerb_Att
 bEnableSubmixSends=false
 bOverrideConcurrency=true
 maxDistance=2097152
 pitchMultiplier=0.75
 soundClassObject=&Fort_Ambient_OneShots_Outdoors
 volumeMultiplier=0.15
 attenuationOverrides {falloffDistance=2500, attenuationShapeExtents {x=500}}
 concurrencyOverrides {maxCount=3}
 soundSubmixSends [{maxSendLevel=8, minSendLevel=0, sendLevel=8, soundSubmix=&Fortnite_Reverb_Default}]
}
