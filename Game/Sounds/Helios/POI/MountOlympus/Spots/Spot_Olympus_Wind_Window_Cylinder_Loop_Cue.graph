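# Looping wind-through-window spot ambience cue for the Mount Olympus POI, using interior window wind attenuation and a voice-limit override.
# A quality switch sends the first two levels to a random pick of four looping waves and the last two to a single shared Mountain wind loop.
# Both branches layer endlessly looping envelopes for slow pitch drift and volume swells.
# A root envelope adds a short fade-in.
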SoundCue /Game/Sounds/Helios/POI/MountOlympus/Spots/Spot_Olympus_Wind_Window_Cylinder_Loop_Cue
root=Enveloper_1
nodes [
 {
  id=Enveloper_0
  children [Random_1]
  properties {
   bLoop=true
   bLoopIndefinitely=true
   loopEnd=30
   pitchCurve {editorCurveData {keys [{interpMode=RCIM_Cubic, time=0}, <at 15.9468 0.9181 RCIM_Cubic>, {interpMode=RCIM_Cubic, time=30}]}}
   volumeCurve {editorCurveData {keys [<at 0 1>]}}
  }
 }
 {
  id=Enveloper_1
  children [QualityLevel_0]
  properties {pitchCurve {editorCurveData {keys [{time=0}]}}, volumeCurve {editorCurveData {keys [<at 0 0 RCIM_Cubic>,<at 0.3 1 RCIM_Cubic>]}}}
 }
 {
  id=Enveloper_2
  children [WavePlayer_4]
  properties {
   bLoop=true
   bLoopIndefinitely=true
   loopEnd=30
   pitchCurve {editorCurveData {keys [{interpMode=RCIM_Cubic, time=0}, <at 15.9468 0.9181 RCIM_Cubic>, {interpMode=RCIM_Cubic, time=30}]}}
   volumeCurve {editorCurveData {keys [<at 0 1>]}}
  }
 }
 {
  id=Enveloper_3
  children [Enveloper_0]
  properties {
   bLoop=true
   bLoopIndefinitely=true
   loopEnd=20
   pitchCurve {editorCurveData {keys [{time=0}]}}
   volumeCurve {editorCurveData {keys [
      {arriveTangent=-0.01, interpMode=RCIM_Cubic, time=0, value=1}
      <at 10 0.8 RCIM_Cubic>
      {interpMode=RCIM_Cubic, leaveTangent=0.01, time=20, value=1}
     ]}}
  }
 }
 {
  id=Enveloper_4
  children [Enveloper_2]
  properties {
   bLoop=true
   bLoopIndefinitely=true
   loopEnd=60
   pitchCurve {editorCurveData {keys [{time=0}]}}
   volumeCurve {editorCurveData {keys [
      {arriveTangent=-0.01, interpMode=RCIM_Cubic, time=0, value=1}
      <at 15 0.8 RCIM_Cubic>
      {arriveTangent=0.0001, interpMode=RCIM_Cubic, leaveTangent=0.0001, tangentMode=RCTM_User, time=30, value=1}
      <at 45 0.7 RCIM_Cubic>
      <at 60 1 RCIM_Cubic>
     ]}}
  }
 }
 {id=QualityLevel_0, children [Enveloper_3,Enveloper_3,Enveloper_4,Enveloper_4]}
 {id=Random_1, children [WavePlayer_1,WavePlayer_2,WavePlayer_3,WavePlayer_5], properties {weights [1,1,1,1]}}
 {
  id=WavePlayer_1
  properties {bLooping=true, soundWaveAssetPtr=/Game/Sounds/Artemis/POI/Hollow_Base/Spots/Spot_Mountain_Wind_Window_Loop_Stereo}
 }
 {id=WavePlayer_2, properties {bLooping=true, soundWaveAssetPtr=&/Spot_Olympus_Window_Wind_01}}
 {id=WavePlayer_3, properties {bLooping=true, soundWaveAssetPtr=&/Spot_Olympus_Window_Wind_02}}
 {
  id=WavePlayer_4
  properties {bLooping=true, soundWaveAssetPtr=&Spot_Mountain_Wind_Window_Loop_Stereo}
 }
 {id=WavePlayer_5, properties {bLooping=true, soundWaveAssetPtr=&/Spot_Olympus_Window_Wind_03}}
]
properties {
 attenuationSettings=/Game/Sounds/Artemis/POI/Hollow_Base/Spots/Spot_Interior_Window_Wind_Att
 bOverrideConcurrency=true
 maxDistance=1253.7231
 soundClassObject=&Fort_Doors
 volumeMultiplier=0.8
 attenuationOverrides {falloffDistance=1600}
 concurrencyOverrides {maxCount=6, voiceStealReleaseTime=2}
}
