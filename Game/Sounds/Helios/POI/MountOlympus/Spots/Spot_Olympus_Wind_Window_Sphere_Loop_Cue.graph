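# Looping spatialized wind ambience cue for window spots at the Mount Olympus POI, with its own attenuation and concurrency limit.
# A random pick among three delay ranges staggers starts, then an enveloper fades the wind in.
# A quality switch gives the first two levels a random pick of four looping waves and the last two one shared mountain-wind loop.
# Nested looping envelopers slowly modulate pitch and volume on both branches.
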SoundCue /Game/Sounds/Helios/POI/MountOlympus/Spots/Spot_Olympus_Wind_Window_Sphere_Loop_Cue
root=Random_0
nodes [
 {id=Delay_0, children [Enveloper_1], properties {delayMax=4}}
 {id=Delay_1, children [Enveloper_1], properties {delayMax=6, delayMin=4}}
 {id=Delay_2, children [Enveloper_1], properties {delayMax=4, delayMin=2}}
 {
  id=Enveloper_0
  children [Random_1]
  properties {
   bLoop=true
   bLoopIndefinitely=true
   loopEnd=30
   pitchCurve {editorCurveData {keys [{interpMode=RCIM_Cubic, time=0}, <at 15.9468 0.9181 RCIM_Cubic>, {interpMode=RCIM_Cubic, time=30}]}}
   volumeCurve {editorCurveData {keys [<at 0 1>]}}
  }
 }
 {
  id=Enveloper_1
  children [QualityLevel_0]
  properties {pitchCurve {editorCurveData {keys [{time=0}]}}, volumeCurve {editorCurveData {keys [<at 0 0 RCIM_Cubic>,<at 2 1 RCIM_Cubic>]}}}
 }
 {
  id=Enveloper_2
  children [WavePlayer_4]
  properties {
   bLoop=true
   bLoopIndefinitely=true
   loopEnd=30
   pitchCurve {editorCurveData {keys [{interpMode=RCIM_Cubic, time=0}, <at 15.9468 0.9181 RCIM_Cubic>, {interpMode=RCIM_Cubic, time=30}]}}
   volumeCurve {editorCurveData {keys [<at 0 1>]}}
  }
 }
 {
  id=Enveloper_3
  children [Enveloper_2]
  properties {
   bLoop=true
   bLoopIndefinitely=true
   loopEnd=60
   pitchCurve {editorCurveData {keys [{time=0}]}}
   volumeCurve {editorCurveData {keys [
      {arriveTangent=-0.01, interpMode=RCIM_Cubic, time=0, value=1}
      <at 15 0.8 RCIM_Cubic>
      {arriveTangent=0.0001, interpMode=RCIM_Cubic, leaveTangent=0.0001, tangentMode=RCTM_User, time=30, value=1}
      <at 45 0.7 RCIM_Cubic>
      <at 60 1 RCIM_Cubic>
     ]}}
  }
 }
 {
  id=Enveloper_4
  children [Enveloper_0]
  properties {
   bLoop=true
   bLoopIndefinitely=true
   loopEnd=60
   pitchCurve {editorCurveData {keys [{time=0}]}}
   volumeCurve {editorCurveData {keys [
      {arriveTangent=-0.01, interpMode=RCIM_Cubic, time=0, value=1}
      <at 15 0.8 RCIM_Cubic>
      {arriveTangent=0.0001, interpMode=RCIM_Cubic, leaveTangent=0.0001, tangentMode=RCTM_User, time=30, value=1}
      <at 45 0.7 RCIM_Cubic>
      <at 60 1 RCIM_Cubic>
     ]}}
  }
 }
 {id=QualityLevel_0, children [Enveloper_4,Enveloper_4,Enveloper_3,Enveloper_3]}
 {id=Random_0, children [Delay_0,Delay_1,Delay_2], properties {weights [1,1,1]}}
 {id=Random_1, children [WavePlayer_0,WavePlayer_1,WavePlayer_2,WavePlayer_3], properties {weights [1,1,1,1]}}
 {
  id=WavePlayer_0
  properties {bLooping=true, soundWaveAssetPtr=/Game/Sounds/Artemis/POI/Hollow_Base/Spots/Spot_Mountain_Wind_Window_Loop_Stereo}
 }
 {id=WavePlayer_1, properties {bLooping=true, soundWaveAssetPtr=&/Spot_Olympus_Window_Wind_01}}
 {id=WavePlayer_2, properties {bLooping=true, soundWaveAssetPtr=&/Spot_Olympus_Window_Wind_02}}
 {id=WavePlayer_3, properties {bLooping=true, soundWaveAssetPtr=&/Spot_Olympus_Window_Wind_03}}
 {
  id=WavePlayer_4
  properties {bLooping=true, soundWaveAssetPtr=&Spot_Mountain_Wind_Window_Loop_Stereo}
 }
]
properties {
 attenuationSettings=/Game/Sounds/Helios/POI/MountOlympus/Spot_Olympus_Window_Wind_Sphere_Att
 bOverrideConcurrency=true
 maxDistance=950
 pitchMultiplier=1.1
 soundClassObject=&Fort_Doors
 volumeMultiplier=0.8
 attenuationOverrides {falloffDistance=1600}
 concurrencyOverrides {maxCount=5, voiceStealReleaseTime=2}
}
