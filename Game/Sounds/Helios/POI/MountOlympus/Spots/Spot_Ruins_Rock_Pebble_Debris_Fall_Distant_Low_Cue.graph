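# Sound cue for distant, low rock and pebble debris falls, an outdoor ambient one-shot spot for the Mount Olympus POI ruins.
# A Random node picks among eighteen equally weighted Asteria Quarry debris-fall waves from the A and B sets, then a Modulator varies pitch.
# A lowered cue pitch and volume turn the reused waves into the distant low variant, with a shared wider one-shot attenuation asset.
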SoundCue /Game/Sounds/Helios/POI/MountOlympus/Spots/Spot_Ruins_Rock_Pebble_Debris_Fall_Distant_Low_Cue
root=Modulator_0
nodes [
 {id=Modulator_0, children [Random_0], properties {pitchMax=1.2, pitchMin=0.8}}
 {
  id=Random_0
  children [
   WavePlayer_0
   WavePlayer_1
   WavePlayer_2
   WavePlayer_3
   WavePlayer_4
   WavePlayer_11
   WavePlayer_12
   WavePlayer_13
   WavePlayer_14
   WavePlayer_15
   WavePlayer_16
   WavePlayer_17
   WavePlayer_18
   WavePlayer_19
   WavePlayer_20
   WavePlayer_21
   WavePlayer_22
   WavePlayer_23
  ]
  properties {weights [1,1,1,1,1,1,1,1,1,1,1,1,1,1,1,1,1,1]}
 }
 {id=WavePlayer_0, properties {soundWaveAssetPtr=/Game/Sounds/Asteria/POI/Quarry/Spots/spot_rock_debris_fall_A_1}}
 {id=WavePlayer_1, properties {soundWaveAssetPtr=/Game/Sounds/Asteria/POI/Quarry/Spots/spot_rock_debris_fall_A_2}}
 {id=WavePlayer_11, properties {soundWaveAssetPtr=/Game/Sounds/Asteria/POI/Quarry/Spots/spot_rock_debris_fall_A_6}}
 {id=WavePlayer_12, properties {soundWaveAssetPtr=/Game/Sounds/Asteria/POI/Quarry/Spots/spot_rock_debris_fall_A_7}}
 {id=WavePlayer_13, properties {soundWaveAssetPtr=/Game/Sounds/Asteria/POI/Quarry/Spots/spot_rock_debris_fall_A_8}}
 {id=WavePlayer_14, properties {soundWaveAssetPtr=/Game/Sounds/Asteria/POI/Quarry/Spots/spot_rock_debris_fall_A_9}}
 {id=WavePlayer_15, properties {soundWaveAssetPtr=/Game/Sounds/Asteria/POI/Quarry/Spots/spot_rock_debris_fall_B_1}}
 {id=WavePlayer_16, properties {soundWaveAssetPtr=/Game/Sounds/Asteria/POI/Quarry/Spots/spot_rock_debris_fall_B_2}}
 {id=WavePlayer_17, properties {soundWaveAssetPtr=/Game/Sounds/Asteria/POI/Quarry/Spots/spot_rock_debris_fall_B_3}}
 {id=WavePlayer_18, properties {soundWaveAssetPtr=/Game/Sounds/Asteria/POI/Quarry/Spots/spot_rock_debris_fall_B_4}}
 {id=WavePlayer_19, properties {soundWaveAssetPtr=/Game/Sounds/Asteria/POI/Quarry/Spots/spot_rock_debris_fall_B_5}}
 {id=WavePlayer_2, properties {soundWaveAssetPtr=/Game/Sounds/Asteria/POI/Quarry/Spots/spot_rock_debris_fall_A_3}}
 {id=WavePlayer_20, properties {soundWaveAssetPtr=/Game/Sounds/Asteria/POI/Quarry/Spots/spot_rock_debris_fall_B_6}}
 {id=WavePlayer_21, properties {soundWaveAssetPtr=/Game/Sounds/Asteria/POI/Quarry/Spots/spot_rock_debris_fall_B_7}}
 {id=WavePlayer_22, properties {soundWaveAssetPtr=/Game/Sounds/Asteria/POI/Quarry/Spots/spot_rock_debris_fall_B_8}}
 {id=WavePlayer_23, properties {soundWaveAssetPtr=/Game/Sounds/Asteria/POI/Quarry/Spots/spot_rock_debris_fall_B_9}}
 {id=WavePlayer_3, properties {soundWaveAssetPtr=/Game/Sounds/Asteria/POI/Quarry/Spots/spot_rock_debris_fall_A_4}}
 {id=WavePlayer_4, properties {soundWaveAssetPtr=/Game/Sounds/Asteria/POI/Quarry/Spots/spot_rock_debris_fall_A_5}}
]
properties {
 attenuationSettings=/Game/Sounds/Attenuation/Fort_TileAmbient_OneShot_Wider_Att
 maxDistance=8000
 pitchMultiplier=0.6
 soundClassObject=&Fort_Ambient_OneShots_Outdoors
 volumeMultiplier=0.301
 attenuationOverrides {bAttenuate=false}
}
